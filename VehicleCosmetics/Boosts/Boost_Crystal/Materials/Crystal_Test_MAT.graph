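# Unlit masked test material for Niagara mesh particles in the Crystal vehicle boost, drawing shiny, fake-depth crystals.
# A vertex-offset chain uses Pivot Painter pivots, UV-stored data and dynamic parameters to scale, rotate and displace pieces, blended with a "Sucky" pull toward object position.
# Emissive layers bump-offset noise and grunge with particle color, a fresnel-darkened core, view-based shine and environment reflections, then passes through MF_FX_EmissiveAdjustments.
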Material /VehicleCosmetics/Boosts/Boost_Crystal/Materials/Crystal_Test_MAT
graph {
 properties {
  bUsedWithNiagaraMeshParticles=true
  blendMode=BLEND_Masked
  pixelDepthOffsetMode=PDOM_Legacy
  shadingModel=MSM_Unlit
  usedShadingModels=Unlit
 }
 nodes [
  MaterialExpressionObjectWorldPosition_0 = ObjectPositionWS()
  n1 = TextureCoordinate() {coordinateIndex=2}
  n2 = ComponentMask(None=TextureCoordinate() {coordinateIndex=3}) {r=true}
  MaterialExpressionMeshEmitterVertexColor_0 = ParticleColor()
  n3 = ComponentMask(None=TextureCoordinate() {coordinateIndex=3}) {g=true}
  n4 = Multiply(A=Constant() {r=1}, B=@n3)
  n5 = Multiply(
   A=Transform(None=AppendVector(A=@n1, B=@n2)) {transformSourceType=TRANSFORMSOURCE_Local}
   B=Multiply(A=Constant() {r=1}, B=Add(A=@MaterialExpressionMeshEmitterVertexColor_0.A, B=@n4))
  )
  n6 = Constant() {r=10}
  n7 = LinearInterpolate(A=Constant3Vector() {constant=<rgba 0 1 0 0>}, B=Constant3Vector() {constant=<rgba 1 1 0 0>}, Alpha=Floor(None=Multiply(A=@n6, B=@n3)))
  n8 = Constant() {r=1}
  MaterialExpressionMeshEmitterDynamicParameter_0 = DynamicParameter() {paramNames [Scale,ScaleRand,RotAmount,Shine]}
  n9 = Multiply(
   A=@n8
   B=LinearInterpolate(
    A=Multiply(A=@n8, B=@MaterialExpressionMeshEmitterDynamicParameter_0.RotAmount)
    B=@MaterialExpressionMeshEmitterDynamicParameter_0.RotAmount
    Alpha=@n3
   )
  )
  MaterialExpressionConstantClamp_2 = Saturate(None=Multiply(A=Frac(None=@n9), B=Constant() {r=6.28}))
  n10 = MaterialFunctionCall() {materialFunction=/Engine/Functions/Engine_MaterialFunctions02/PivotPainter/PivotPainter_PerObjectData}
  n11 = Constant()
  n12 = LinearInterpolate(A=@n11, B=@n4, Alpha=@MaterialExpressionMeshEmitterDynamicParameter_0.ScaleRand)
  n13 = Multiply(
   A=LinearInterpolate(A=Constant() {r=-1}, B=Add(A=@n11, B=@n12), Alpha=@MaterialExpressionMeshEmitterDynamicParameter_0.Scale)
   B=Subtract(A=WorldPosition().XYZ, B=@n10.Pivot Position)
  )
  n14 = RotateAboutAxis(
   NormalizedRotationAxis=AppendVector(A=Normalize(VectorInput=@n7), B=@MaterialExpressionConstantClamp_2)
   RotationAngle=@n10.Pivot Position
   PivotPoint=@n10.Pivot Position
   Position=Add(A=@n13, B=WorldPosition().XYZ)
  )
  n15 = Add(A=Distance(A=WorldPosition().XYZ, B=@MaterialExpressionObjectWorldPosition_0), B=Multiply(A=Constant() {r=1}, B=Constant() {r=2048}))
  n16 = Constant() {r=1}
  MaterialExpressionConstantClamp_1 = Saturate(None=Power(Base=Divide(A=@n15, B=@n16), Exp=ScalarParameter() {defaultValue=12, group=Sucky, parameterName=SuckyGradFalloff}))
  LinearInterpolate(
   A=Subtract(A=@MaterialExpressionObjectWorldPosition_0, B=WorldPosition().XYZ)
   B=Add(A=Add(A=@n5, B=@n14), B=@n13)
   Alpha=@MaterialExpressionConstantClamp_1
  )
  n17 = TextureCoordinate() {uTiling=3, vTiling=3}
  n18 = Constant() {r=6}
  MaterialExpressionConstantClamp_4 = Saturate(None=TextureSample(UVs=BumpOffset(Coordinate=@n17, Height=@n18)) {texture=/DealershipSharedLibrary/FX/Textures/Noise/T_VCVFX_NoiseGeneric01Pack_RGB}.RGB)
  n19 = TextureCoordinate() {uTiling=0.5, vTiling=0.5}
  n20 = Constant() {r=-5}
  n21 = Constant() {r=-3}
  n22 = Multiply(
   A=TextureSample(UVs=BumpOffset(Coordinate=@n19, Height=@n20)) {texture=/VehicleCosmetics/Boosts/Boost_Crystal/Textures/T_VCVFX_Boost_Crystal_GrungeOverlay_CLR}.RGB
   B=TextureSample(UVs=BumpOffset(Coordinate=@n19, Height=@n21)) {texture=&T_VCVFX_Boost_Crystal_GrungeOverlay_CLR}.RGB
  )
  n23 = Constant() {r=3}
  n24 = Multiply(A=Multiply(A=@MaterialExpressionConstantClamp_4, B=Constant() {r=32}), B=Multiply(A=Power(Base=@n22, Exp=@n23), B=Constant() {r=16}))
  n25 = Multiply(A=@n22, B=Constant() {r=1})
  MaterialExpressionMeshEmitterVertexColor_1 = ParticleColor()
  n26 = Multiply(A=Constant() {r=1}, B=@MaterialExpressionMeshEmitterVertexColor_1.RGB)
  MaterialExpressionConstantClamp_3 = Saturate(None=Add(A=Add(A=@n24, B=@n25), B=Multiply(A=@n26, B=@MaterialExpressionMeshEmitterVertexColor_1.RGB)))
  n27 = DotProduct(A=CameraVectorWS(), B=Constant3Vector() {constant=<rgba 0 0 1 0>})
  MaterialExpressionConstantClamp_5 = Saturate(None=Power(Base=@n27, Exp=Constant() {r=20}))
  n28 = Constant() {r=5}
  n29 = MaterialFunctionCall!(Desaturate=@n28, "Tint (Curve)"=@n28, Tint Amount=Constant() {r=1}, Cubemap) {materialFunction=/DealershipSharedLibrary/FX/Materials/MaterialFunctions/VC_MF_WorldEnv}
  n30 = Constant() {r=0.01}
  n31 = Multiply(A=ScalarParameter() {defaultValue=0.012, parameterName=CoreDarkness}, B=Constant3Vector() {constant=<rgba 1 0.3 0.1 0>})
  MaterialExpressionMeshEmitterVertexColor_2 = ParticleColor()
  n32 = Fresnel(Normal=StaticSwitchParameter(True=Constant3Vector() {constant=<rgba 0 0 1 0>}, False=VertexNormalWS()) {parameterName=NoAnim}) {exponent=10}
  n33 = Add(
   A=Multiply(A=@MaterialExpressionMeshEmitterDynamicParameter_0.Shine, B=@MaterialExpressionConstantClamp_5)
   B=Add(A=Multiply(A=@n29.WorldSpace, B=@n30), B=LinearInterpolate(A=@n31, B=@MaterialExpressionMeshEmitterVertexColor_2.RGB, Alpha=@n32))
  )
  n34 = ReflectionVectorWS()
  n35 = Multiply(
   A=TextureSample(UVs=ComponentMask(None=@n34) {g=true, r=true}) {texture=/VehicleCosmetics/Boosts/Boost_Crystal/Textures/T_VCVFX_Boost_Crystal_Boxy_CubeNegY}.G
   B=Constant() {r=0.225}
  )
  MaterialFunctionCall(Emissive=Add(A=Multiply(A=@MaterialExpressionConstantClamp_3, B=@n33), B=Multiply(A=@MaterialExpressionMeshEmitterVertexColor_2.RGB, B=@n35))) {materialFunction=&MF_FX_EmissiveAdjustments}
  MaterialFunctionCall()
  MaterialExpressionConstantClamp_0 = Saturate(None=ConstantBiasScale(None=OneMinus(None=@n27)) {bias=0.03, scale=15})
  Reroute(None=@MaterialExpressionConstantClamp_0)
 ]
}
