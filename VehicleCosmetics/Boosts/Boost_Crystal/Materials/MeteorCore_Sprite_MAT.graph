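# Masked unlit Niagara sprite material for the meteor core in the Crystal boost effect.
# Panned water normals, some in radial coordinates, distort UVs into a wobbly disc.
# Squared distance with steep saturates forms a hard cutout and hot rim, both shrinking as particle alpha drops.
# Emissive blends thresholded scrolling noise in the interior to particle colour, overbright at the rim, through MF_FX_EmissiveAdjustments.
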Material /VehicleCosmetics/Boosts/Boost_Crystal/Materials/MeteorCore_Sprite_MAT
graph {
 properties {bUsedWithNiagaraSprites=true, blendMode=BLEND_Masked, pixelDepthOffsetMode=PDOM_Legacy, shadingModel=MSM_Unlit, usedShadingModels=Unlit}
 nodes [
  n1 = ComponentMask(None=TextureSample() {samplerType=SAMPLERTYPE_Normal, texture=/DealershipSharedLibrary/FX/Textures/Normal/T_VCVFX_Radial_Coords01_D}.RGB) {g=true, r=true}
  n2 = Panner(Coordinate=Multiply(A=@n1, B=Constant2Vector() {g=0.1, r=1})) {speedX=0.01, speedY=-0.3}
  n3 = Panner(Coordinate=Multiply(A=@n1, B=Constant2Vector() {g=0.25, r=2})) {speedX=-0.015, speedY=-0.3}
  n4 = ComponentMask(None=Add(
    A=TextureSample(UVs=@n2) {samplerType=SAMPLERTYPE_Normal, texture=/DealershipSharedLibrary/FX/Textures/Normal/T_VCVFX_Water03Swizzled_Noise}.RGB
    B=TextureSample(UVs=@n3) {samplerType=SAMPLERTYPE_Normal, texture=&T_VCVFX_Water03Swizzled_Noise}.RGB
   )) {g=true, r=true}
  n5 = Multiply(A=@n4, B=Constant() {r=0.05})
  n6 = TextureCoordinate() {uTiling=1.5, vTiling=1.5}
  n7 = Panner(Coordinate=TextureCoordinate() {uTiling=0.2, vTiling=0.2}) {speedX=-0.015, speedY=-0.3}
  n8 = Multiply(
   A=ComponentMask(None=TextureSample(UVs=@n7) {samplerType=SAMPLERTYPE_Normal, texture=&T_VCVFX_Water03Swizzled_Noise}.RGB) {g=true, r=true}
   B=Constant() {r=0.1}
  )
  n9 = Add(A=Add(A=Add(A=@n5, B=@n6), B=@n8), B=Constant() {r=-0.75})
  n10 = DotProduct(A=@n9, B=@n9)
  MaterialExpressionVertexColor_0 = ParticleColor()
  n11 = Add(A=@n10, B=OneMinus(None=@MaterialExpressionVertexColor_0.A))
  n12 = OneMinus(None=@n11)
  n13 = Constant() {r=0.6}
  n14 = Constant() {r=256}
  MaterialExpressionConstantClamp_6 = Saturate(None=Multiply(A=Subtract(A=@n12, B=@n13), B=@n14))
  n15 = TextureSample(UVs=Panner() {speedX=-0.1, speedY=-0.05}) {texture=/DealershipSharedLibrary/FX/Textures/Noise/T_VCVFX_NoiseGeneric01Pack_RGB}
  n16 = TextureSample(UVs=Panner(Coordinate=TextureCoordinate()) {speedX=0.1, speedY=0.1}) {texture=&T_VCVFX_NoiseGeneric01Pack_RGB}
  MaterialExpressionConstantClamp_9 = Saturate(None=Multiply(A=Multiply(A=@n15.B, B=@n16.B), B=Constant() {r=256}))
  MaterialExpressionConstantClamp_10 = Saturate(None=Multiply(A=@n10, B=Constant() {r=8}))
  n17 = Multiply(A=@MaterialExpressionVertexColor_0.RGB, B=Constant() {r=5})
  n18 = Subtract(A=@n13, B=Constant() {r=0.25})
  MaterialExpressionConstantClamp_8 = Saturate(None=Multiply(A=Subtract(A=@n11, B=@n18), B=@n14))
  n19 = TextureCoordinate() {uTiling=2, vTiling=2}
  n20 = Constant() {r=0.5}
  n21 = Add(A=Add(A=Add(A=@n19, B=@n8), B=Multiply(A=@n4, B=@n20)), B=Constant() {r=-1})
  MaterialExpressionConstantClamp_11 = Saturate(None=Power(Base=DotProduct(A=@n21, B=@n21), Exp=Constant() {r=3}))
  MaterialExpressionConstantClamp_7 = Saturate(None=Add(A=@MaterialExpressionConstantClamp_11, B=@MaterialExpressionConstantClamp_8))
  MaterialFunctionCall(Emissive=LinearInterpolate(
    A=Multiply(A=@MaterialExpressionConstantClamp_9, B=@MaterialExpressionConstantClamp_10)
    B=LinearInterpolate(A=@MaterialExpressionVertexColor_0.RGB, B=@n17, Alpha=@MaterialExpressionConstantClamp_8)
    Alpha=@MaterialExpressionConstantClamp_7
   )) {materialFunction=&MF_FX_EmissiveAdjustments}
 ]
}
